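; Local gameplay behavior that has an AI bot place and tend Avocado Eater eggs on the MotherGround.
; When the prior egg has finished hatching, a random roll lets the bot emote, trace to ground ahead and spawn a surface-aligned egg.
; An unhatched, untriggered egg gets a tending emote; both paths send a Tend.Start gameplay event.
; Egg damage or death turns a nearby non-friendly instigator hostile if the bot is Unaware.
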(event LocalGameplayBehavior|EventOnExecute
  (params (Activator :type Actor%20Object%20Reference))
  (Utilities|IsValid (Variables|Internal|GetSpawnedEggActor)
    (:Is%20Valid
      (if (Variables|/Game/Athena/Items/Consumables/AvacadoEater/BGA_Athena_AvacadoEater_Deployed_Parent.BGA_Athena_AvacadoEater_Deployed_Parent_C|GetHatchFinished? (Variables|Internal|GetSpawnedEggActor))
        (if (<= (Math|Random|RandomFloatinRange 0.0 1.0) (Variables|Setup|GetChanceToPlaceAvocado))
          (bind _asfort_athena_aibot_controller (Utilities|Casting|CastToFortAthenaAIBotController Activator)
            (:then
              (Variables|Internal|SetBotController _asfort_athena_aibot_controller)
              (Utilities|FlowControl|Sequence
                (:then_0
                  (CallFunction|SetFocusRotation (Variables|Internal|GetBotController)))
                (:then_1
                  (Utilities|FlowControl|Delay (Variables|Setup|GetStartDelayNoEggs))
                  (Class|FortAthenaAIBotController|PlayEmote (Variables|Internal|GetBotController) (Variables|Setup|GetEmoteWhenNoEggs))
                  (Utilities|FlowControl|Delay (Math|Random|RandomFloatinRange (Variables|Setup|GetPlacementTimeMin) (Variables|Setup|GetPlacementTimeMax)))
                  (bind (_outhit _returnvalue) (Collision|LineTraceByProfile :Start (+ (+ (* (Transformation|GetActorForwardVector (Variables|/Script/FortniteAI.FortAthenaAIBotController|GetPlayerBotPawn (Variables|Internal|GetBotController))) (ScalableFloat|GetValueatLevel (Variables|Internal|GetAvacadoOffset))) (Transformation|GetActorLocation (Variables|/Script/FortniteAI.FortAthenaAIBotController|GetPlayerBotPawn (Variables|Internal|GetBotController)))) "0, 0,100.000000") :End (- (+ (* (Transformation|GetActorForwardVector (Variables|/Script/FortniteAI.FortAthenaAIBotController|GetPlayerBotPawn (Variables|Internal|GetBotController))) (ScalableFloat|GetValueatLevel (Variables|Internal|GetAvacadoOffset))) (Transformation|GetActorLocation (Variables|/Script/FortniteAI.FortAthenaAIBotController|GetPlayerBotPawn (Variables|Internal|GetBotController)))) "0, 0,100.000000") :ProfileName "FortSupplyDropGroundTrace" :bTraceComplex false :DrawDebugType "None" :bIgnoreSelf true :TraceColor "(R=1.000000,G=0.000000,B=0.000000,A=1.000000)" :TraceHitColor "(R=0.000000,G=1.000000,B=0.000000,A=1.000000)" :DrawTime 3e+01))
                  (if _returnvalue
                    (bind _returnvalue_2 (Game|SpawnActorfromClass (Variables|Setup|GetActorToSpawn) (Math|Transform|MakeTransform (.Location (Collision|BreakHitResult _outhit)) (Math|Rotator|MakeRotfromZX (.Normal (Collision|BreakHitResult _outhit)) (Math|Vector|RotateVectorAroundAxis "1.000000, 0, 0" (Variables|Internal|GetRandomAngle) "0, 0,1.000000"))) "AlwaysSpawn" "SelectDefaultAtRuntime" 0 (Variables|/Script/FortniteAI.FortAthenaAIBotController|GetPlayerBotPawn (Variables|Internal|GetBotController))))
                    (bind _asbga_athena_avacado_eater_deployed (Utilities|Casting|CastToBGA_Athena_AvacadoEater_Deployed _returnvalue_2)
                      (:then
                        (bind _output_get (Variables|Internal|SetSpawnedEggActor _asbga_athena_avacado_eater_deployed))
                        (EventDispatchers|BindEventtoOnDied _output_get (delegate Custom|OnEggDied))
                        (EventDispatchers|BindEventtoOnDamaged _output_get (delegate Custom|OnDamaged)))
                      (:CastFailed
                        _))
                    (else
                      _))
                  (Utilities|FlowControl|Delay 0.1)
                  (CallFunction|StartTending)
                  (Utilities|FlowControl|Delay (Variables|Setup|GetTimeAfterPlacement))
                  (bind _botcontroller (Variables|Internal|GetBotController)
                    (:then
                      (Class|FortAthenaAIBotController|StopEmote _botcontroller)
                      (CallFunction|ResetFocusRotation _botcontroller)
                      (LocalGameplayBehavior|FinishExecute))
                    (:else)))))
            (:CastFailed
              _))
          (else
            _))
        (LocalGameplayBehavior|FinishExecute)
        (elif (Variables|/Game/Athena/Items/Consumables/AvacadoEater/BGA_Athena_AvacadoEater_Deployed_Parent.BGA_Athena_AvacadoEater_Deployed_Parent_C|GetHatchHasTriggered? (Variables|Internal|GetSpawnedEggActor))
          (LocalGameplayBehavior|FinishExecute)
          (else
            (bind _asfort_athena_aibot_controller_2 (Utilities|Casting|CastToFortAthenaAIBotController Activator))
            (Variables|Internal|SetBotController _asfort_athena_aibot_controller_2)
            (Utilities|FlowControl|Sequence
              (:then_0
                _)
              (:then_1
                ;[1 1 1] If there is an Egg placed and the Egg has not hatched, play a Tending Animation.
                (Utilities|FlowControl|Delay (Variables|Setup|GetStartDelayWhenEggs))
                (Class|FortAthenaAIBotController|PlayEmote (Variables|Internal|GetBotController) (Variables|Setup|GetEmoteWhenEggs))
                (Utilities|FlowControl|Sequence
                  (:then_0
                    (Utilities|FlowControl|Delay 0.5)
                    (CallFunction|StartTending))
                  (:then_1
                    (Utilities|FlowControl|Delay (Variables|Setup|GetTendingWaitTime))
                    (bind _botcontroller_2 (Variables|Internal|GetBotController)
                      (:then
                        (Class|FortAthenaAIBotController|StopEmote _botcontroller_2)
                        (CallFunction|ResetFocusRotation _botcontroller_2)
                        (LocalGameplayBehavior|FinishExecute))
                      (:else))))))))))
    (:Is%20Not%20Valid
      _)))

(event LocalGameplayBehavior|EventOnAbort
  (bind _botcontroller (Variables|Internal|GetBotController)
    (:then
      (Class|FortAthenaAIBotController|StopEmote _botcontroller)
      (CallFunction|ResetFocusRotation _botcontroller))
    (:else)))

(event Custom|OnDamaged
  (params (DamagedActor :type Actor%20Object%20Reference) (Damage :type Float%20%28single-precision%29) (InstigatedBy :type Controller%20Object%20Reference) (DamageCauser :type Actor%20Object%20Reference) (HitLocation :type Vector) (FHitComponent :type Primitive%20Component%20Object%20Reference) (BoneName :type Name) (Momentum :type Vector))
  (CallFunction|FindDamageCauserandSetHostile InstigatedBy) ; We check both OnDamaged and OnDied because the Egg has a LifeSpan where the OnDied can be called a few seconds after the egg has actually beend estroyed.
)

(event Custom|OnEggDied
  (params (DamagedActor :type Actor%20Object%20Reference) (Damage :type Float%20%28single-precision%29) (InstigatedBy :type Controller%20Object%20Reference) (DamageCauser :type Actor%20Object%20Reference) (HitLocation :type Vector) (FHitComponent :type Primitive%20Component%20Object%20Reference) (BoneName :type Name) (Momentum :type Vector))
  ;[1 1 1] If Egg got smashed by a Vehicle
  (Variables|Internal|SetSpawnedEggActor 0)
  (Utilities|IsValid DamageCauser
    (:Is%20Valid
      (Utilities|Casting|CastToFortAthenaVehicle DamageCauser)
      (Utilities|IsValid InstigatedBy
        (:Is%20Valid
          (bind _asplayer_controller (Utilities|Casting|CastToPlayerController InstigatedBy))
          (CallFunction|FindDamageCauserandSetHostile _asplayer_controller))))))

(fn FindDamageCauserAndSetHostile
  (params (Controller_DamageCauser :type Controller%20Object%20Reference))
  (bind _returnvalue (Pawn|GetControlledPawn Controller_DamageCauser))
  (bind _botcontroller (Variables|Internal|GetBotController)
    (:then
      (if (and (<= (Math|Vector|Distance2D(Vector) (Transformation|GetActorLocation _returnvalue) (Transformation|GetActorLocation (Variables|/Script/FortniteAI.FortAthenaAIBotController|GetPlayerBotPawn _botcontroller))) (ScalableFloat|GetValueatLevel (Variables|Internal|GetDistanceToAggro))) (Utilities|Enum|Equal(Enum) (Variables|/Script/FortniteAI.FortAthenaAIBotController|GetCurrentAlertLevel _botcontroller) "Unaware"))
        (if (not (Utilities|Enum|Equal(Enum) (Team|GetTeamAffiliation _returnvalue (Variables|/Script/FortniteAI.FortAthenaAIBotController|GetPlayerBotPawn _botcontroller)) "Friendly"))
          (Affiliation|ChangeAttitudeBetweenActors (Variables|/Script/FortniteAI.FortAthenaAIBotController|GetPlayerBotPawn _botcontroller) _returnvalue "Hostile"))))
    (:else)))

(fn StartTending
  (bind _botcontroller (Variables|Internal|GetBotController))
  (bind _playerbotpawn (Variables|/Script/FortniteAI.FortAthenaAIBotController|GetPlayerBotPawn _botcontroller))
  (Utilities|IsValid _botcontroller
    (:Is%20Valid
      (Utilities|IsValid _playerbotpawn
        (:Is%20Valid
          (Ability|SendGameplayEventtoActor _playerbotpawn "(TagName=\"MotherGround.Events.Tend.Start\")" (Utilities|Struct|MakeGameplayEventData "(TagName=\"\")" 0 0 (Variables|Internal|GetSpawnedEggActor))))))))

(fn ConstructionScript)
